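; Seasonal firework projectile for the Frozen Fireworks quest interactable that bursts one second after spawn, or sooner on hit.
; Both triggers repeat the same logic: stop sparks and movement, fire the Niagara explosion, hide the firework, then kill the actor after a delay.
; A curve-table hotfix read at BeginPlay can switch off the explosion sound, which otherwise plays only when the first local controller is the instigator's.
; Debug prints remain on spawn and hit.
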(event Custom|Kaboom
  (Components|Activation|Deactivate (Variables|Default|GetP_FrozenFirework_Sparks))
  (Components|Activation|Activate (Variables|Default|GetNiagaraExplosion) true)
  (Game|Components|ProjectileMovement|StopSimulating (Variables|Movement|GetProjectileMovementComponent) (Collision|MakeHitResult false))
  (Development|SetHiddeninGame (Variables|Default|GetFirework) true false)
  (Utilities|Time|SetTimerbyEvent (delegate Custom|Cleanup) 1e+01)
  (switch int (Variables|Default|GetHotFixSoundOnOff)
    (:0
      (if (Utilities|Equal(Object) (Utilities|Array|Get(acopy) (Fortnite|GetLocalFortPlayerControllers true)) (Pawn|GetController (Variables|Actor|GetInstigator)))
        (Audio|PlaySoundatLocation "/Game/Sounds/Fort_Weapons/Launchers/Fireworks/Firework_Launcher_Explosion_Cue.Firework_Launcher_Explosion_Cue" (Transformation|GetActorLocation)))
      (Rendering|SetVisibility (Variables|Default|GetFirework) false false))
    (:1
      _)))

(event Collision|EventHit
  (params (MyComp :type Primitive%20Component%20Object%20Reference) (Other :type Actor%20Object%20Reference) (OtherComp :type Primitive%20Component%20Object%20Reference) (bSelfMoved :type Boolean) (HitLocation :type Vector :default "0, 0, 0") (HitNormal :type Vector :default "0, 0, 0") (NormalImpulse :type Vector :default "0, 0, 0") (Hit :type Hit%20Result%20Structure%20%28by%20ref%29))
  (Development|PrintString "I Bumped" false false)
  (Utilities|Time|ClearTimerbyFunctionName 0 "Kaboom")
  (Components|Activation|Deactivate (Variables|Default|GetP_FrozenFirework_Sparks))
  (Components|Activation|Activate (Variables|Default|GetNiagaraExplosion) true)
  (Game|Components|ProjectileMovement|StopSimulating (Variables|Movement|GetProjectileMovementComponent) (Collision|MakeHitResult false))
  (Development|SetHiddeninGame (Variables|Default|GetFirework) true false)
  (Utilities|Time|SetTimerbyEvent (delegate Custom|Cleanup) 1e+01)
  (switch int (Variables|Default|GetHotFixSoundOnOff)
    (:0
      (if (Utilities|Equal(Object) (Utilities|Array|Get(acopy) (Fortnite|GetLocalFortPlayerControllers true)) (Pawn|GetController (Variables|Actor|GetInstigator)))
        (Audio|PlaySoundatLocation "/Game/Sounds/Fort_Weapons/Launchers/Fireworks/Firework_Launcher_Explosion_Cue.Firework_Launcher_Explosion_Cue" (Transformation|GetActorLocation)))
      (Rendering|SetVisibility (Variables|Default|GetFirework) false false))
    (:1
      _)))

(event Custom|Cleanup
  (Delivery|Kill))

(event EventBeginPlay
  (Utilities|Time|SetTimerbyFunctionName 0 "Kaboom" 1.0)
  (Development|PrintString "FIRED" false false)
  (bind _outxy (Fortnite|EvaluateCurveTableRow (Variables|Default|GetSoundOffHotfix)))
  (bind _output_get (Variables|Default|SetHotFixSoundOnOff (Math|Float|Truncate _outxy)))
  (Development|PrintString (Utilities|String|ToString(Integer) _output_get)))

(event Collision|EventActorBeginOverlap
  (params (OtherActor :type Actor%20Object%20Reference)))

(event EventTick
  (params (DeltaSeconds :type Float%20%28single-precision%29)))

(fn ConstructionScript)
